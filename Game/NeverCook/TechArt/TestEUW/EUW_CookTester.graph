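# Editor utility widget in a NeverCook TechArt test folder, named as a cook tester, that exposes its own properties in a details panel.
# On PreConstruct it passes itself to a DetailsView inside a StackBox.
# Two variables default to hard references to cosmetic assets: an Athena dance item and a shoes item.
# No other logic is wired; Construct and Tick are empty.
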EditorUtilityWidgetBlueprint /Game/NeverCook/TechArt/TestEUW/EUW_CookTester
graphs [EventGraph {nodes [
   Event PreConstruct()
   > SetObject(self=Get DetailsView_54() -> (DetailsView_54:Details View Object Reference), NewObject=Self-Reference())
   Event Construct()
   Event Tick()
  ]}]
variables [
 EID {type=AthenaDanceItemDefinition, default=/BRCosmetics/Athena/Items/Cosmetics/Dances/EID_Mime}
 ShoeTest {type=CosmeticShoesItemDefinition, default=/CosmeticShoes/Assets/Items/Cosmetics/Shoes_BreathTalePeony}
]
widgets {
 info {rootWidgetClass=&StackBox, widgetCount=2}
 widgets [
  StackBox_0 {widgetClassPath=&StackBox, widget=StackBox_0}
  DetailsView_54 {widgetClassPath=&DetailsView, parent=StackBox_0, widget=DetailsView_54, slot {class=StackBoxSlot}}
 ]
}
